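; Blueprint for the GimmickFemale emote effect that fades a skeletal mesh in with a glitch look.
; On BeginPlay it zeroes the mesh's OpacityMult, waits briefly, prints a debug string, then plays a fade-in timeline.
; Timeline opacity and glitch tracks lerp material scalars for opacity, world position offset strength, color offset width and discoloration.
; Tick and overlap events are empty, and the construction script only calls its parent.
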(event Collision|EventActorBeginOverlap
  (params (OtherActor :type Actor%20Object%20Reference)))

(event EventTick
  (params (DeltaSeconds :type Float%20%28single-precision%29)))

(event EventBeginPlay
  (bind _skeletalmesh (Variables|Default|GetSkeletalMesh))
  (|CallParentFunction)
  (Rendering|Material|SetScalarParameterValueonMaterials (Variables|Default|GetSkeletalMesh) "OpacityMult" 0.0)
  (Utilities|FlowControl|Delay 0.75)
  (Development|PrintString "Fade In Start")
  (bind (_direction _time_opacity _time_glitch) (Timeline|Timeline_Fade%20in PlayFromStart)
    (:Update
      (Rendering|Material|SetScalarParameterValueonMaterials _skeletalmesh "OpacityMult" (Math|Float|Lerp 0.0 4.0 _time_opacity))
      (Rendering|Material|SetScalarParameterValueonMaterials _skeletalmesh "WPOStrength" (Math|Float|Lerp 5.0 1e+02 _time_glitch))
      (Rendering|Material|SetScalarParameterValueonMaterials _skeletalmesh "Color Offset WIdth" (Math|Float|Lerp 0.0 0.004 _time_glitch))
      (Rendering|Material|SetScalarParameterValueonMaterials _skeletalmesh "Use Discolor" (Math|Float|Lerp 0.0 0.3 _time_glitch)))))

(fn ConstructionScript
  (|CallParentFunction))
